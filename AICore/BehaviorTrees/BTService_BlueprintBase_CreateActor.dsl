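; Behavior tree service Blueprint that spawns an actor at the controlled pawn while its branch is active.
; On activation it spawns the tunable actor class at the pawn's transform and stores the reference in an internal variable.
; On deactivation it destroys the stored actor if it is still valid and the tunable destroy-on-cease-relevant flag is set.
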(event AI|EventReceiveActivationAI
  (params (OwnerController :type AIController%20Object%20Reference) (ControlledPawn :type Pawn%20Object%20Reference))
  (bind _returnvalue (Game|SpawnActorfromClass (Variables|Tunable|GetActorClass) (Transformation|GetActorTransform ControlledPawn) "Undefined" "MultiplyWithRoot" 0 0))
  (Variables|Internal|SetActor _returnvalue))

(event AI|EventReceiveDeactivationAI
  (params (OwnerController :type AIController%20Object%20Reference) (ControlledPawn :type Pawn%20Object%20Reference))
  (bind _actor (Variables|Internal|GetActor))
  (if (Variables|tunable|GetbDestroyOnCeaseRelevant)
    (Utilities|IsValid _actor
      (:Is%20Valid
        (Actor|DestroyActor _actor)))))
