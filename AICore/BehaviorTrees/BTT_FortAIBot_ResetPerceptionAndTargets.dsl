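; Behavior tree task that clears a Fortnite Athena AI bot's perception and current targets.
; On execute it casts the owner to FortAthenaAIBotController and calls ResetPerceptionandTargets with UpdateNow, then finishes successfully.
; If the cast fails, the task finishes as failed.
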(event EventReceiveExecute
  (params (OwnerActor :type Actor%20Object%20Reference))
  (bind _asfort_athena_aibot_controller (Utilities|Casting|CastToFortAthenaAIBotController OwnerActor)
    (:then
      (Class|FortAthenaAIBotController|ResetPerceptionandTargets _asfort_athena_aibot_controller "UpdateNow")
      (AI|BehaviorTree|FinishExecute true))
    (:CastFailed
      (AI|BehaviorTree|FinishExecute false))))
