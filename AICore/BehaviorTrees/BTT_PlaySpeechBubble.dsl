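; Behavior tree task that makes an AI-controlled pawn play a speech bubble chosen by a gameplay tag on the blackboard.
; On execute, it reads the tag container from the configured blackboard key, breaks it into an array and casts the pawn to FortPawn.
; It passes the first tag to the shared BPFLAICommonFunctions PlaySpeechBubble helper and finishes successfully.
; A failed cast or empty container finishes the task as failed.
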(event AI|EventReceiveExecuteAI
  (params (OwnerController :type AIController%20Object%20Reference) (ControlledPawn :type Pawn%20Object%20Reference))
  (bind _speechbubbletagkey (Variables|Default|GetSpeechBubbleTagKey))
  (bind _returnvalue (AI|BehaviorTree|GetBlackboardValueasGameplayTag _speechbubbletagkey))
  (bind _gameplaytags (GameplayTags|BreakGameplayTagContainer _returnvalue))
  (bind _asfort_pawn (Utilities|Casting|CastToFortPawn ControlledPawn)
    (:then
      (if (> (Utilities|Array|Length _gameplaytags) 0)
        (Class|BPFLAICommonFunctions|PlaySpeechBubble _asfort_pawn (Utilities|Array|Get(aref) _gameplaytags) false true)
        (AI|BehaviorTree|FinishExecute true)
        (else
          _)))
    (:CastFailed
      _))
  (AI|BehaviorTree|FinishExecute false))
