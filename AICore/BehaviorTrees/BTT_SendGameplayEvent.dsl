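; Behavior Tree task that sends gameplay events to the AI controller actor, rather than its controlled pawn.
; It casts the controller's brain component to a Behavior Tree component and uses it with GameplayEventTag to fetch a blackboard tag container.
; Each tag in that container is sent as a separate event.
; The task finishes successfully only if at least one event is sent.
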(event AI|EventReceiveExecuteAI
  (params (OwnerController :type AIController%20Object%20Reference) (ControlledPawn :type Pawn%20Object%20Reference))
  (Variables|Default|SetTaskSucceeded false)
  (bind _asbehavior_tree_component (Utilities|Casting|CastToBehaviorTreeComponent (Variables|/Script/AIModule.AIController|GetBrainComponent OwnerController))
    (:then
      (for _array_element (GameplayTags|BreakGameplayTagContainer (Blackboard|GetTagContainer (Variables|Default|GetGameplayEventTag) _asbehavior_tree_component))
        (Ability|SendGameplayEventtoActor OwnerController _array_element)
        (Variables|Default|SetTaskSucceeded true)))
    (:CastFailed
      _))
  (AI|BehaviorTree|FinishExecute (Variables|Default|GetTaskSucceeded)))
