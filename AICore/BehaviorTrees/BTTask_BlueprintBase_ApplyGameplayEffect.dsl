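; Blueprint base behavior tree task that applies a configurable Gameplay Effect to the AI's own actor.
; On execute, it casts the SelfActor blackboard value to Actor and gets its Ability System Component.
; A valid component applies the effect class variable to itself at level 0 and finishes the task with success.
; A trailing Finish Execute reports failure when the cast or component lookup does not succeed.
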(event EventReceiveExecute
  (params (OwnerActor :type Actor%20Object%20Reference))
  (bind _asactor (Utilities|Casting|CastToActor (AI|Components|Blackboard|GetValueasObject (AI|GetBlackboard OwnerActor) (Utilities|String|StringToName "SelfActor")))
    (:then
      (Utilities|IsValid (Ability|GetAbilitySystemComponent _asactor)
        (:Is%20Valid
          (GameplayEffects|ApplyGameplayEffectToTarget :self (Ability|GetAbilitySystemComponent _asactor) :GameplayEffectClass (Variables|Default|GetGameplay%20Effect%20Class) :Target (Ability|GetAbilitySystemComponent _asactor) :Level 0.0)
          (AI|BehaviorTree|FinishExecute true))
        (:Is%20Not%20Valid
          _)))
    (:CastFailed
      _))
  (AI|BehaviorTree|FinishExecute false))
